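# Niagara VFX system for the SwoopClasp player emote, layering timed glows, debris and line bursts on a skeletal mesh.
# Spawn group numbers double as bone/socket indices, with group 4 given its own color, size and timing.
# Base sprites, debris and mesh streaks follow those groups, while multidirectional glows anchor to index 5.
# Front-end preview, time-of-day and ColorIndex inputs adjust brightness or tint, and emitter age switches streaks into despawn behavior.
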NiagaraSystem /BRCosmetics/Animation/Game/MainPlayer/Emotes/SwoopClasp/VFX/NS_Emote_SwoopClasp
dependencies {
 usedDataInterfaces [
  &NiagaraDataInterfaceCurve
  &NiagaraDataInterfacePlatformSet
  &NiagaraDataInterfaceArrayFloat
  &NiagaraDataInterfaceVectorCurve
  &NiagaraDataInterfaceSkeletalMesh
  &NiagaraDataInterfaceColorCurve
 ]
 usedDynamicInputs [
  &SelectFloatFromArray
  &MakeVector2D
  &HueShiftLinearColor
  &ColorFromCurve
 ]
 usedModules [
  &self:SystemSpawnScript.SystemScriptSource.SystemScriptGraph.NiagaraNodeAssignment_0.SetVariables_#1
  &self:SystemSpawnScript.SystemScriptSource.SystemScriptGraph.NiagaraNodeAssignment_2.SetVariables_#2
  &self:SystemSpawnScript.SystemScriptSource.SystemScriptGraph.NiagaraNodeAssignment_3.SetVariables_#3
  &SystemState
  &self:SystemSpawnScript.SystemScriptSource.SystemScriptGraph.NiagaraNodeAssignment_1.SetVariables_#4
 ]
}
emitters {
 Base {
  emitterUpdateScript [
   &EmitterState(
    Inactive Response=::NewEnumerator1
    Life Cycle Mode=::NewEnumerator1
    Loop Behavior=::NewEnumerator1
    Loop Duration Mode=::NewEnumerator1
    Scalability Mode=::NewEnumerator1
    UseLoopDelay=false
   )
   &SpawnBurst_Instantaneous(Spawn Group=4, Spawn Time=1.7, Use Spawn Probability=false, UseLoopCountLimit=true)
   &SpawnBurst_Instantaneous(Spawn Group=3, Spawn Time=1.57, Use Spawn Probability=false, UseLoopCountLimit=true) {module=SpawnBurst_Instantaneous004}
   &SpawnBurst_Instantaneous(Spawn Group=2, Spawn Time=1.425, Use Spawn Probability=false, UseLoopCountLimit=true) {module=SpawnBurst_Instantaneous003}
   &SpawnBurst_Instantaneous(Spawn Group=1, Spawn Time=1.723, Use Spawn Probability=false, UseLoopCountLimit=true) {module=SpawnBurst_Instantaneous002}
   &SpawnBurst_Instantaneous(Spawn Time=1.685, Use Spawn Probability=false, UseLoopCountLimit=true) {module=SpawnBurst_Instantaneous001}
   &SpawnRate(Use Spawn Probability=false) {enabled=false, module=SpawnRate001}
  ]
  particleSpawnScript [
   &InitializeParticle(
    Color=&MaskLinearColorBySpawnGroup(
     Masked Linear Color=&RandomRangeLinearColor(Minimum=<rgba 1.8329 0.2536 2 0.8>, Maximum=<rgba 2 0.3045 1.543 0.8>)
     Passthrough Linear Color=&RandomRangeLinearColor(Minimum=<rgba 0.2536 2 1.8545 0.8>, Maximum=<rgba 0.3045 2 0.8696 0.8>)
     Spawn Group Mask=4
    )
    Lifetime=&MaskFloatBySpawnGroup(Masked Float=2.3, Passthrough Float=5, Spawn Group Mask=4)
    Sprite Size Mode=::NewEnumerator3
    Uniform Sprite Size=&MaskFloatBySpawnGroup(Masked Float=90, Passthrough Float=60, Spawn Group Mask=4)
    UsePositionOffset=false
    Write Material Random=true
   )
   &SphereLocation(
    MaskSpawnGroup=false
    Offset=<0 0 7>
    Override Seed=false
    Sphere Radius=0
    Surface Only=false
    UseNonUniformScale=false
    UseOffset=true
    UseSphereOrientation=false
   ) {enabled=false}
   /NiagaraFN/Sprite/FN_SpriteAlignmentAndFacing(SpriteAlignment:Vector3f=<0 1 0>, SpriteFacing:Vector3f=<0 0 1>)
   &DynamicMaterialParameters(
    Index 0 Param 1=&RandomRangeFloat(Minimum=0.7, Maximum=0.8)
    Index 0 Param 2=0.4
    Index 0 Param 3=1
    Index 0 Param 4=0.004
    Module.Param0WriteEnabled=true
    Module.Param1WriteEnabled=true
    Module.Param2WriteEnabled=false
    Module.Param3WriteEnabled=true
   )
   &DynamicMaterialParameters(
    Index 0 Param 1=&RandomRangeFloat(Minimum=0.7, Maximum=0.8)
    Index 0 Param 2=0.4
    Index 0 Param 3=1
    Index 0 Param 4=0.004
    Module.Param0WriteEnabled=true
    Module.Param1WriteEnabled=true
    Module.Param2WriteEnabled=false
    Module.Param3WriteEnabled=true
   ) {module=DynamicMaterialParameters001}
   /SetVariables(
    Particles.CameraGlowColor:LinearColor=&MaskLinearColorBySpawnGroup(
     Masked Linear Color=&RandomRangeLinearColor(Minimum=<rgba 1.8329 0.2536 2 0.35>, Maximum=<rgba 2 0.3045 1.543 0.35>)
     Passthrough Linear Color=&RandomRangeLinearColor(Minimum=<rgba 0.2536 2 1.8545 0.25>, Maximum=<rgba 0.3045 2 0.8696 0.25>)
     Spawn Group Mask=4
    )
    Particles.CameraGlowDynam:Vector4f=&MakeVector4(X=&RandomRangeFloat(Minimum=0.7, Maximum=0.8), Y=1.0039, Z=0, W=1)
    Particles.GlowSize=&MaskVector2dBySpawnGroup(Masked Vector2D=<80 60>, Passthrough Vector2D=<50 50>, Spawn Group Mask=4)
   ) {isSetParameters=true}
  ]
  particleUpdateScript [
   &SkeletalMeshLocation(
    Bone / Socket Index=&SpawnGroup()
    Bone Sampling Mode=::NewEnumerator4
    MaskSpawnGroup=false
    Orientation Sampling=::NewEnumerator1
    Sampled Position Offset=<0 0 -6.5>
    Skeletal Mesh=$System.Skeletal Mesh
    Use Sampled Mesh Attribute as Ribbon ID=false
   ) {module=SkeletalMeshLocation001}
   &SkeletalMeshLocation(
    Bone / Socket Index=5
    Bone Sampling Mode=::NewEnumerator4
    MaskSpawnGroup=false
    Orientation Sampling=::NewEnumerator1
    Skeletal Mesh=$System.Skeletal Mesh
    Use Sampled Mesh Attribute as Ribbon ID=false
    Write To Particle Position=false
   ) {module=SkeletalMeshLocation002}
   /SetVariables(Particles.DistanceToTable:NiagaraFloat=&NormalizeDistanceRange(
     Start Position:NiagaraPosition=&MakeVector(
      X=0
      Y=0
      Z=&MakeFloatFromVector(Vector=Linked() {linkedVariable {name=Particles.SkeletalMeshLocation002.SampledPosition, type=NiagaraPosition}}, Channel=::NewEnumerator2)
     )
     End Position:NiagaraPosition=&MakeVector(
      X=0
      Y=0
      Z=&MakeFloatFromVector(Vector=Linked() {linkedVariable {name=Particles.Position, type=NiagaraPosition}}, Channel=::NewEnumerator2)
     )
     Distance:NiagaraFloat=2
     Invert Normalized Range:NiagaraBool=true
    )) {isSetParameters=true}
   &ParticleState()
   &ScaleSpriteSize(
    Scale Sprite Size Mode=::NewEnumerator0
    Uniform Scale Factor=&FloatFromCurve(FloatCurve=DataInterface() {propertyValues {Curve {
        defaultValue=0
        keys [{interpMode=RCIM_Cubic, tangentMode=RCTM_User, time=0, value=0}, {interpMode=RCIM_Cubic, tangentMode=RCTM_User, time=0.1, value=1}]
       }}})
   )
   &SolveForcesAndVelocity(Clamp Velocity=false, Limit Acceleration=false)
   &ScaleColor(
    Color Value To Scale=$Particles.Initial.CameraGlowColor
    Scale Alpha=&FloatFromCurve(
     FloatCurve=DataInterface() {propertyValues {Curve {
        defaultValue=1
        keys [
         <at -0.0147 0 RCIM_Cubic>
         <at 0.195 1.0064>
         {arriveTangent=-0.0405, interpMode=RCIM_Cubic, leaveTangent=-0.0405, tangentMode=RCTM_User, time=0.7, value=1}
         {arriveTangent=-10.0927, interpMode=RCIM_Cubic, leaveTangent=-10.0927, tangentMode=RCTM_User, time=1, value=0}
        ]
       }}}
     Scale Curve=&MakeCustomFloatFromBool(Bool=$User.bIsFrontEndPreview, True Float=0.3, False Float=0.6)
    )
    ScaleA=true
    ScaleRGB=true
   )
   &ScaleColor(
    Scale Alpha=&FloatFromCurve(
     FloatCurve=DataInterface() {propertyValues {Curve {
        defaultValue=1
        keys [
         <at -0.0147 0 RCIM_Cubic>
         <at 0.195 1.0064>
         {arriveTangent=-0.0405, interpMode=RCIM_Cubic, leaveTangent=-0.0405, tangentMode=RCTM_User, time=0.7, value=1}
         {arriveTangent=-5.9574, interpMode=RCIM_Cubic, leaveTangent=-5.9574, tangentMode=RCTM_User, time=1, value=0}
        ]
       }}}
     Scale Curve=0.5
    )
    ScaleA=true
    ScaleRGB=true
   ) {module=ScaleColor001}
  ]
  rendererClasses [&NiagaraSpriteRendererProperties]
  renderers [
   {rendererClass=&NiagaraSpriteRendererProperties}
   {rendererClass=&NiagaraSpriteRendererProperties, rendererIndex=1}
  ]
 }
 Debris {
  emitterUpdateScript [
   &EmitterState(
    Inactive Response=::NewEnumerator1
    Life Cycle Mode=::NewEnumerator1
    Loop Behavior=::NewEnumerator1
    Loop Delay=3.85
    Loop Duration=1
    UseLoopDelay=true
   )
   &SpawnBurst_Instantaneous(Spawn Count=30, Spawn Time=0.05, Use Spawn Probability=false)
   &SpawnBurst_Instantaneous(Spawn Count=1, Spawn Group=1, Spawn Time=0.05, Use Spawn Probability=false) {module=SpawnBurst_Instantaneous001}
   &SpawnBurst_Instantaneous(Spawn Count=2, Spawn Group=4, Use Spawn Probability=false, UseLoopCountLimit=false) {module=SpawnBurst_Instantaneous002}
  ]
  particleSpawnScript [
   &InitializeParticle(
    Color=&MaskLinearColorBySpawnGroup(
     Masked Linear Color=<rgba 1 1 1 1>
     Passthrough Linear Color=&RandomRangeLinearColor(Minimum=<rgba 1.8329 0.2536 2 1>, Maximum=<rgba 2 0.3045 1.543 1>)
     Spawn Group Mask=4
    )
    Lifetime=&MaskFloatBySpawnGroup(
     Masked Float=&RandomRangeFloat(Minimum=0.4, Maximum=0.455)
     Passthrough Float=&MaskFloatBySpawnGroup(Masked Float=0.16, Passthrough Float=&RandomRangeFloat(Minimum=0.3, Maximum=1), Spawn Group Mask=1)
     Spawn Group Mask=4
    )
    Mesh Scale Mode=::NewEnumerator3
    Mesh Uniform Scale=&RandomRangeFloat(Minimum=0.4, Maximum=0.45)
    Sprite Rotation Mode=::NewEnumerator2
    Sprite Size=&Multiply_Vector2D_ByFloat(
     Float=&RandomRangeFloat(Minimum=0.1, Maximum=0.3)
     Vector2D=&RandomRangeVector2D(Minimum=<30 30>, Maximum=<40 40>)
    )
    Sprite Size Mode=::NewEnumerator4
    UsePositionOffset=false
    Write Material Random=true
   ) {module=InitializeParticle001}
   &SkeletalMeshLocation(
    Bone / Socket Index=&SpawnGroup()
    Bone Sampling Mode=::NewEnumerator4
    MaskSpawnGroup=false
    Position Sampling=::NewEnumerator2
    Skeletal Mesh=$System.Skeletal Mesh
    Use Sampled Mesh Attribute as Ribbon ID=false
   ) {module=SkeletalMeshLocation001}
   &ShapeLocation(
    Cylinder Height=0
    Cylinder Radius=&MaskFloatBySpawnGroup(Masked Float=0, Passthrough Float=15, Spawn Group Mask=4)
    MaskSpawnGroup=false
    Offset=&MaskVectorBySpawnGroup(Masked Vector=<0 0 5>, Passthrough Vector=<0 0 -10>, Spawn Group Mask=4)
    Offset Coordinate Space=::Local
    Offset Mode=::NewEnumerator0
    Override Seed=false
    Shape Origin=$Particles.SkeletalMeshLocation001.SampledPosition
    Shape Primitive=::NewEnumerator1
    Surface Only=false
    UseOwnerScale=true
   )
   &AddVelocityFromPoint(
    Module.Use Falloff Distance=false
    Origin Offset Coordinate Space=::World
    Velocity Strength=&RandomRangeFloat(Minimum=10, Maximum=30)
   )
   &AddVelocity(
    Cone Angle=180
    Cone Axis=<0 0 1>
    Module.UseConeFalloff=false
    Module.UseVelDistribution=false
    Override Seed=false
    Velocity Mode=::NewEnumerator2
    Velocity Speed=&RandomRangeFloat(Minimum=30, Maximum=60)
   ) {module=AddVelocity001}
   &AddVelocity(Velocity=&RandomRangeVector(Minimum=<-50 -50 0>, Maximum=<50 -50 100>)) {enabled=false}
   &InitialMeshOrientation(
    Mesh Orientation Mode=::NewEnumerator7
    Rotation=&RandomRangeVector(Minimum=<-0.3 -0.3 -1>, Maximum=<0.3 0.3 1>)
    Use Rotation Vector=true
   )
   &InitialMeshOrientation(Mesh Orientation Mode=::NewEnumerator7, Rotation=<0 0.25 0>, Use Rotation Vector=true) {module=InitialMeshOrientation001}
   &CameraOffset(Camera Offset Amount=&MaskFloatBySpawnGroup(Masked Float=15, Passthrough Float=30, Spawn Group Mask=4))
   /SetVariables(Particles.VisibilityTag=&SpawnGroup()) {isSetParameters=true}
   /SetVariables(Particles.Velocity=&MaskVectorBySpawnGroup(Masked Vector=<0 0 0>, Passthrough Vector=$Particles.Velocity, Spawn Group Mask=4)) {isSetParameters=true}
  ]
  particleUpdateScript [
   &ParticleState()
   &Drag(Drag=&RandomRangeFloat(Minimum=1, Maximum=2), Use Linear Drag=true, Use Rotational Drag=false)
   &ScaleMeshSize(Scale Factor=&VectorFromFloat(Value=&FloatFromCurve(
      FloatCurve=DataInterface() {propertyValues {Curve {keys [
          {arriveTangent=1.9115, interpMode=RCIM_Cubic, leaveTangent=1.9115, tangentMode=RCTM_User, time=0, value=0}
          {arriveTangent=0.0224, interpMode=RCIM_Cubic, leaveTangent=0.0224, tangentMode=RCTM_User, time=1, value=1}
         ]}}}
      Scale Curve=&MakeCustomFloatFromBool(Bool=$User.bIsFrontEndPreview, True Float=0.35, False Float=0.5)
     )))
   &ScaleMeshSize(Scale Factor=&RandomRangeVector(Minimum=<1 1 1.5>, Maximum=<1 1 2>)) {module=ScaleMeshSize001}
   &DynamicMaterialParameters(
    Index 0 Param 2=&RandomRangeFloat(Minimum=0, Maximum=0)
    Index 0 Param 4=13
    Index 1 Param 1=&FloatFromCurve(FloatCurve=DataInterface() {propertyValues {Curve {keys [<at 0 0.5>,<at 1 1>]}}})
    Index 1 Param 2=2
    Index 1 Param 3=&FloatFromCurve(FloatCurve=DataInterface() {propertyValues {Curve {keys [{interpMode=RCIM_Cubic, tangentMode=RCTM_Break, time=0.5, value=-0.5}, <at 1 0.1>]}}})
    Index 2 Param 1=&Multiply_Float(
     A=&FloatFromCurve(FloatCurve=DataInterface() {propertyValues {Curve {keys [{interpMode=RCIM_Cubic, leaveTangent=8.1085, tangentMode=RCTM_Break, time=0.8568, value=1}, {tangentMode=RCTM_User, time=1, value=0}]}}})
     B=&Divide_Float(
      A=&MakeFloatFromVector(Vector=$Particles.Scale, Channel=::NewEnumerator0)
      B=&RandomRangeFloat(Minimum=10, Maximum=15)
     )
    )
    Index 2 Param 2=&FloatFromCurve(
     FloatCurve=DataInterface() {propertyValues {Curve {keys [
         {interpMode=RCIM_Cubic, tangentMode=RCTM_Break, time=0.7, value=0}
         {arriveTangent=-0.0859, interpMode=RCIM_Cubic, tangentMode=RCTM_Break, time=1, value=-1}
        ]}}}
     Scale Curve=50
    )
    Index 2 Param 3=&Waveform(
     Global Input Position=&RandomRangeFloat(Minimum=0, Maximum=15)
     Global Phase=$Particles.Age
     "Waveform [1]"=::NewEnumerator12
     "[1] Phase Scale"=2
     "[1] Frequency"=1
     "[1] Amplitude Min/Max"=<0 1>
    )
    Index 2 Param 4=25
    Index 3 Param 1=&FloatFromCurve(
     FloatCurve=DataInterface() {propertyValues {Curve {keys [
         {interpMode=RCIM_Cubic, leaveTangent=-0.0025, tangentMode=RCTM_Break, time=0, value=0}
         {arriveTangent=1.8528, interpMode=RCIM_Cubic, tangentMode=RCTM_Break, time=1, value=1}
        ]}}}
     Scale Curve=&RandomRangeFloat(Minimum=0, Maximum=0)
    )
    Index 3 Param 2=&FloatFromCurve(
     FloatCurve=DataInterface() {propertyValues {Curve {keys [
         {interpMode=RCIM_Cubic, leaveTangent=-0.0025, tangentMode=RCTM_Break, time=0, value=0.5}
         {arriveTangent=-0.0316, interpMode=RCIM_Cubic, tangentMode=RCTM_Break, time=1, value=-0.2}
        ]}}}
     Scale Curve=0
    )
    Index 3 Param 3=&FloatFromCurve(FloatCurve=DataInterface() {propertyValues {Curve {keys [
         {interpMode=RCIM_Cubic, leaveTangent=1.1034, tangentMode=RCTM_Break, time=0, value=-1}
         {arriveTangent=2.2399, tangentMode=RCTM_User, time=0.5, value=0.125}
        ]}}})
    Index 3 Param 4=&FloatFromCurve(FloatCurve=DataInterface() {propertyValues {Curve {keys [
         {interpMode=RCIM_Cubic, leaveTangent=1.8089, tangentMode=RCTM_Break, time=0.3, value=-0.5}
         {arriveTangent=2.0757, interpMode=RCIM_Cubic, tangentMode=RCTM_Break, time=0.9, value=1}
        ]}}})
    Module.Index0Param0Write=true
    Module.Index0Param1Write=true
    Module.Index0Param2Write=true
    Module.Index0Param3Write=true
    Module.Index1Param0Write=true
    Module.Index1Param1Write=true
    Module.Index1Param2Write=true
    Module.Index1Param3Write=true
    Module.Index2Param0Write=true
    Module.Index2Param1Write=true
    Module.Index2Param2Write=true
    Module.Index2Param3Write=true
    Module.Index3Param0Write=true
    Module.Index3Param1Write=true
    Module.Index3Param2Write=true
    Module.Index3Param3Write=true
    Write Parameter Index 1=true
    Write Parameter Index 2=true
    Write Parameter Index 3=true
   )
   &SolveForcesAndVelocity(Clamp Velocity=false, Limit Acceleration=false)
   &SpriteRotationRate(Rotation Rate=&RandomRangeFloat(Minimum=1500, Maximum=500)) {enabled=false, module=SpriteRotationRate001}
   &ScaleColor(
    Scale Alpha=&MaskFloat(
     Mask Source=0
     Mask Parameter=?
     Mask=4
     Masked Float=&FloatFromCurve(FloatCurve=DataInterface() {propertyValues {Curve {keys [
          {tangentMode=RCTM_User, time=0, value=1}
          {arriveTangent=-4.3986, interpMode=RCIM_Cubic, leaveTangent=-4.3986, tangentMode=RCTM_User, time=0.45, value=1}
          {arriveTangent=0.1205, interpMode=RCIM_Cubic, leaveTangent=0.1205, tangentMode=RCTM_User, time=1, value=0}
         ]}}})
     Passthrough Float=1
    )
    Scale RGB=&MaskVectorBySpawnGroup(
     Masked Vector=&Lerp_Vector(
      A=&VectorFromFloat(Value=&FloatFromCurve(FloatCurve=DataInterface() {propertyValues {Curve {keys [<at 0.2771 1>,<at 0.3786 1>]}}}, Scale Curve=20))
      B=&VectorFromFloat(Value=&RandomRangeFloat(Minimum=1, Maximum=2.5, Evaluation Type=::NewEnumerator1))
      Alpha=&FloatFromCurve(FloatCurve=DataInterface() {propertyValues {Curve {keys [
           {interpMode=RCIM_Cubic, leaveTangent=2.5294, tangentMode=RCTM_Break, time=0, value=0}
           {arriveTangent=-0.2166, interpMode=RCIM_Cubic, tangentMode=RCTM_Break, time=0.4, value=1}
          ]}}})
     )
     Passthrough Vector=&Multiply_VectorByFloat(
      Vector=&VectorFromFloat(Value=&RandomRangeFloat(Minimum=0.4, Maximum=1))
      Float=&FloatFromCurve(FloatCurve=DataInterface() {propertyValues {Curve {keys [<at 0.2 1>,<at 0.3 1>,<at 1 0>]}}})
     )
     Spawn Group Mask=4
    )
    ScaleA=true
    ScaleRGB=true
   )
   &ScaleSpriteSize(Scale Factor=&Vector2DFromFloat(Value=&FloatFromCurve(
      FloatCurve=DataInterface() {propertyValues {Curve {keys [{interpMode=RCIM_Cubic, tangentMode=RCTM_User, time=0.7, value=1}, {interpMode=RCIM_Cubic, tangentMode=RCTM_User, time=1, value=0}]}}}
      Scale Curve=0.6
     )))
   &FN_Light_Attributes(
    Color=&MultiplyLinearColor(
     Linear Color=&LerpLinearColors(StartColor=<rgba 1 0.0515 0.4969 1>, EndColor=<rgba 1 0.0515 0.0515 1>, LerpFactor=$User.ColorIndex)
     Scale Factor=&MakeVector4_FromLinearColor(LinearColor=$System.TODColor)
    )
    Create Color Attribute=true
    Create Exponent Attribute=true
    Create Position Attribute=true
    Create Radius Attribute=true
    Exponent=2
    Position=&AddVectorToPosition(Position=$Particles.SkeletalMeshLocation001.SampledPosition, Vector=<0 -20 0>)
    Radius=30
    Scale Color=&Multiply_Float(
     A=&Multiply_Float(
      A=10
      B=&FloatFromCurve(FloatCurve=DataInterface() {propertyValues {Curve {
          defaultValue=1
          keys [
           {tangentMode=RCTM_User, time=0, value=1}
           {arriveTangent=-0.0405, interpMode=RCIM_Cubic, leaveTangent=-0.0405, tangentMode=RCTM_User, time=0.7009, value=1}
           {arriveTangent=-8.0089, interpMode=RCIM_Cubic, leaveTangent=-8.0089, tangentMode=RCTM_User, time=1, value=0}
          ]
         }}})
     )
     B=&FNTimeOfDayScalar(Night Scale=&Lerp_Float(A=1, B=2.5, Alpha=$User.ColorIndex), Morning Scale=1, Evening Scale=1)
    )
   )
  ]
  rendererClasses [
   &NiagaraSpriteRendererProperties
   &NiagaraLightRendererProperties
   &NiagaraMeshRendererProperties
  ]
  renderers [
   {rendererClass=&NiagaraSpriteRendererProperties}
   {rendererClass=&NiagaraLightRendererProperties, rendererIndex=1}
   {rendererClass=&NiagaraMeshRendererProperties, rendererIndex=2}
   {rendererClass=&NiagaraMeshRendererProperties, rendererIndex=3}
  ]
 }
 Line_Burst001 {
  emitterUpdateScript [
   &EmitterState(Life Cycle Mode=::NewEnumerator1, Loop Behavior=::NewEnumerator1, Loop Duration=9, UseLoopDelay=false)
   &SpawnRate(
    SpawnGroup=4
    SpawnRate=&FloatFromCurve(
     FloatCurve=DataInterface() {propertyValues {Curve {keys [<at 0 0 RCIM_Constant>, {interpMode=RCIM_Cubic, tangentMode=RCTM_User, time=1.725, value=300}, <at 2.175 10 RCIM_Constant>, <at 3.8 0>]}}}
     CurveIndex=$Emitter.Age
     Scale Curve=&FN_FrontEndFloatModifier(Gameplay=1, Front End=1)
    )
    Use Spawn Probability=false
   ) {module=SpawnRate001}
   &SpawnRate(
    SpawnRate=&FloatFromCurve(
     FloatCurve=DataInterface() {propertyValues {Curve {keys [
         <at 0 0 RCIM_Constant>
         {interpMode=RCIM_Cubic, tangentMode=RCTM_User, time=1.685, value=300}
         {arriveTangent=-85.5422, interpMode=RCIM_Constant, time=2.135, value=5}
         <at 3.8 0>
        ]}}}
     CurveIndex=$Emitter.Age
     Scale Curve=&FN_FrontEndFloatModifier(Gameplay=1, Front End=1)
    )
    Use Spawn Probability=false
   ) {module=SpawnRate004}
   &SpawnRate(
    SpawnGroup=1
    SpawnRate=&FloatFromCurve(
     FloatCurve=DataInterface() {propertyValues {Curve {keys [
         <at 0 0 RCIM_Constant>
         {interpMode=RCIM_Cubic, tangentMode=RCTM_User, time=1.723, value=300}
         {arriveTangent=-85.5422, interpMode=RCIM_Constant, time=2.17, value=5}
         <at 3.8 0>
        ]}}}
     CurveIndex=$Emitter.Age
     Scale Curve=&FN_FrontEndFloatModifier(Gameplay=1, Front End=1)
    )
    Use Spawn Probability=false
   )
   &SpawnRate(
    SpawnGroup=2
    SpawnRate=&FloatFromCurve(
     FloatCurve=DataInterface() {propertyValues {Curve {keys [
         <at 0 0 RCIM_Constant>
         {interpMode=RCIM_Cubic, tangentMode=RCTM_User, time=1.425, value=300}
         {arriveTangent=-85.5422, interpMode=RCIM_Constant, time=1.89, value=5}
         <at 3.8 0>
        ]}}}
     CurveIndex=$Emitter.Age
     Scale Curve=&FN_FrontEndFloatModifier(Gameplay=1, Front End=1)
    )
    Use Spawn Probability=false
   ) {module=SpawnRate002}
   &SpawnRate(
    SpawnGroup=3
    SpawnRate=&FloatFromCurve(
     FloatCurve=DataInterface() {propertyValues {Curve {keys [
         <at 0 0 RCIM_Constant>
         {interpMode=RCIM_Cubic, tangentMode=RCTM_User, time=1.57, value=300}
         {arriveTangent=-85.5422, interpMode=RCIM_Constant, time=2.035, value=5}
         <at 3.8 0>
        ]}}}
     CurveIndex=$Emitter.Age
     Scale Curve=&FN_FrontEndFloatModifier(Gameplay=1, Front End=1)
    )
    Use Spawn Probability=false
   ) {module=SpawnRate003}
   &SpawnBurst_Instantaneous(
    Spawn Count=&MakeCustomIntFromBool(Bool=$User.bIsFrontEndPreview, True INT=36, False INT=&RandomRangeInt(Minimum=18, Maximum=24))
    Spawn Group=4
    Spawn Time=3.95
    Use Spawn Probability=false
    UseLoopCountLimit=false
   )
   &SpawnRate(
    SpawnRate=&FloatFromCurve(
     FloatCurve=DataInterface() {propertyValues {Curve {keys [
         <at 0 0 RCIM_Constant>
         <at 6.7 100 RCIM_Cubic>
         {arriveTangent=-398.8572, interpMode=RCIM_Cubic, leaveTangent=-398.8572, tangentMode=RCTM_User, time=6.95, value=0}
        ]}}}
     CurveIndex=$Emitter.Age
     Scale Curve=&FN_FrontEndFloatModifier(Gameplay=1, Front End=1)
    )
    Use Spawn Probability=false
   ) {module=SpawnRate005}
   &SpawnRate(
    SpawnGroup=1
    SpawnRate=&FloatFromCurve(
     FloatCurve=DataInterface() {propertyValues {Curve {keys [
         <at 0 0 RCIM_Constant>
         <at 6.6 100 RCIM_Cubic>
         {arriveTangent=-398.8572, interpMode=RCIM_Cubic, leaveTangent=-398.8572, tangentMode=RCTM_User, time=6.85, value=0}
        ]}}}
     CurveIndex=$Emitter.Age
     Scale Curve=&FN_FrontEndFloatModifier(Gameplay=1, Front End=1)
    )
    Use Spawn Probability=false
   ) {module=SpawnRate006}
   &SpawnRate(
    SpawnGroup=2
    SpawnRate=&FloatFromCurve(
     FloatCurve=DataInterface() {propertyValues {Curve {keys [
         <at 0 0 RCIM_Constant>
         <at 6.7 100 RCIM_Cubic>
         {arriveTangent=-398.8572, interpMode=RCIM_Cubic, leaveTangent=-398.8572, tangentMode=RCTM_User, time=6.95, value=0}
        ]}}}
     CurveIndex=$Emitter.Age
     Scale Curve=&FN_FrontEndFloatModifier(Gameplay=1, Front End=1)
    )
    Use Spawn Probability=false
   ) {module=SpawnRate007}
   &SpawnRate(
    SpawnGroup=3
    SpawnRate=&FloatFromCurve(
     FloatCurve=DataInterface() {propertyValues {Curve {keys [
         <at 0 0 RCIM_Constant>
         <at 6.65 100 RCIM_Cubic>
         {arriveTangent=-398.8572, interpMode=RCIM_Cubic, leaveTangent=-398.8572, tangentMode=RCTM_User, time=6.9, value=0}
        ]}}}
     CurveIndex=$Emitter.Age
     Scale Curve=&FN_FrontEndFloatModifier(Gameplay=1, Front End=1)
    )
    Use Spawn Probability=false
   ) {module=SpawnRate008}
   /SetVariables(Emitter.IsDespawning:NiagaraBool=&SetBoolByFloatComparison(A=$Emitter.Age, B=3.9, Comparison Type=::NewEnumerator2)) {isSetParameters=true}
  ]
  particleSpawnScript [
   &InitializeParticle(
    Color=&MaskLinearColorBySpawnGroup(
     Masked Linear Color=&RandomRangeLinearColor(Minimum=<rgba 1.8329 0.2536 2 1>, Maximum=<rgba 2 0.3045 1.543 1>)
     Passthrough Linear Color=&RandomRangeLinearColor(Minimum=<rgba 0.2536 2 1.8545 1>, Maximum=<rgba 0.3045 2 0.8696 1>)
     Spawn Group Mask=4
    )
    Lifetime Max=&MakeCustomFloatFromBool(Bool=$Emitter.IsDespawning, True Float=0.45, False Float=0.2)
    Lifetime Min:NiagaraFloat=&MakeCustomFloatFromBool(Bool=$Emitter.IsDespawning, True Float=0.05, False Float=0.15)
    Lifetime Mode=::NewEnumerator1
    Mesh Scale Max=&MakeCustomVectorFromBool(Bool=$Emitter.IsDespawning, True Vector=<0.075 0.075 0.6>, False Vector=<0.375 0.375 1>)
    Mesh Scale Min=&MakeCustomVectorFromBool(Bool=$Emitter.IsDespawning, True Vector=<0.05 0.05 0.5>, False Vector=<0.2 0.2 0.8>)
    Mesh Scale Mode=::NewEnumerator2
    UsePositionOffset=false
    Write Material Random=true
   )
   &SkeletalMeshLocation(
    Bone / Socket Index=&SpawnGroup()
    Bone Sampling Mode=::NewEnumerator4
    MaskSpawnGroup=false
    Position Sampling=::NewEnumerator2
    Skeletal Mesh=$System.Skeletal Mesh
    Use Sampled Mesh Attribute as Ribbon ID=false
   ) {module=SkeletalMeshLocation001}
   &ShapeLocation(
    Cylinder Height=0
    Cylinder Radius=&MaskFloatBySpawnGroup(
     Masked Float=&MakeCustomFloatFromBool(Bool=$Emitter.IsDespawning, True Float=15, False Float=19)
     Passthrough Float=6
     Spawn Group Mask=4
    )
    MaskSpawnGroup=false
    Offset=<0 0 -10>
    Offset Coordinate Space=::Local
    Offset Mode=::NewEnumerator0
    Override Seed=false
    Shape Origin=$Particles.SkeletalMeshLocation001.SampledPosition
    Shape Primitive=::NewEnumerator1
    Surface Only=false
    UseOwnerScale=true
   ) {module=ShapeLocation001}
   &CameraOffset(Camera Offset Amount=15)
   &AddVelocity(Velocity=&MakeCustomVectorFromBool(
     Bool=$Emitter.IsDespawning
     True Vector=&RandomRangeVector(Minimum=<-10 -10 90>, Maximum=<10 10 130>)
     False Vector=&RandomRangeVector(Minimum=<0 0 50>, Maximum=<0 0 100>)
    ))
   /SetVariables(Particles.Velocity=&Multiply_VectorByFloat(
     Vector=$Particles.Velocity
     Float=&FloatFromCurve(
      FloatCurve=DataInterface() {propertyValues {Curve {
         defaultValue=1
         keys [
          {tangentMode=RCTM_User, time=0, value=1}
          {arriveTangent=-0.0405, interpMode=RCIM_Cubic, leaveTangent=-0.0405, tangentMode=RCTM_User, time=0.2, value=1}
          {arriveTangent=-0.0309, interpMode=RCIM_Cubic, leaveTangent=-0.0309, tangentMode=RCTM_User, time=1, value=0}
         ]
        }}}
      CurveIndex=$Emitter.NormalizedLoopAge
     )
    )) {isSetParameters=true}
   /SetVariables(Particles.VisibilityTag=&MakeCustomIntFromBool(Bool=$Emitter.IsDespawning, True INT=1, False INT=0)) {isSetParameters=true}
  ]
  particleUpdateScript [
   &ParticleState()
   &ScaleMeshSize(Scale Factor=&MakeCustomVectorFromBool(
     Bool=$Emitter.IsDespawning
     True Vector=&VectorFromCurve(VectorCurve=DataInterface() {propertyValues {XCurve {keys [<at 0 1>,<at 1 1>]}, YCurve {keys [<at 0 1>,<at 1 1>]}, ZCurve {keys [<at 0 0>,<at 1 1>]}}})
     False Vector=&Multiply_VectorByFloat(
      Vector=&MakeVector(
       X=&FloatFromCurve(
        FloatCurve=DataInterface() {propertyValues {Curve {
           defaultValue=1
           keys [
            {tangentMode=RCTM_User, time=0, value=1}
            {arriveTangent=-3.2935, interpMode=RCIM_Cubic, leaveTangent=-3.2935, tangentMode=RCTM_User, time=0.2, value=1}
            {arriveTangent=-0.3234, interpMode=RCIM_Cubic, leaveTangent=-0.3234, tangentMode=RCTM_User, time=1, value=0.4}
           ]
          }}}
        CurveIndex=$Emitter.NormalizedLoopAge
       )
       Y=&FloatFromCurve(
        FloatCurve=DataInterface() {propertyValues {Curve {
           defaultValue=1
           keys [
            {tangentMode=RCTM_User, time=0, value=1}
            {arriveTangent=-3.2929, interpMode=RCIM_Cubic, leaveTangent=-3.2929, tangentMode=RCTM_User, time=0.2, value=1}
            {arriveTangent=-0.1937, interpMode=RCIM_Cubic, leaveTangent=-0.1937, tangentMode=RCTM_User, time=1, value=0.4}
           ]
          }}}
        CurveIndex=$Emitter.NormalizedLoopAge
       )
       Z=&FloatFromCurve(FloatCurve=DataInterface() {propertyValues {Curve {keys [<at 0 1>,<at 0.7 1>,<at 1 0>]}}})
      )
      Float=&Multiply_Float(
       A=0.1
       B=&FloatFromCurve(
        FloatCurve=DataInterface() {propertyValues {Curve {keys [
            {arriveTangent=4.1785, time=0, value=0}
            <at 0.5 1.1>
            {arriveTangent=-0.0405, interpMode=RCIM_Cubic, leaveTangent=-0.0405, tangentMode=RCTM_User, time=1, value=1.5}
           ]}}}
        Scale Curve=&Multiply_Float(
         A=1
         B=&ClampFloat(
          Float=&FN_LOD_Distance_Remap(Start Distance=200, End Distance=400, Minimum Output=0.5, Maximum Output=1)
          Min=0.5
         )
        )
       )
      )
     )
    ))
   &Drag(
    Drag=&MakeCustomFloatFromBool(Bool=$Emitter.IsDespawning, True Float=&RandomRangeFloat(Minimum=2, Maximum=3), False Float=5)
    Use Linear Drag=true
    Use Rotational Drag=true
   )
   &SolveForcesAndVelocity(Clamp Velocity=false, Limit Acceleration=false)
   &ScaleColor(
    Scale Alpha=&FloatFromCurve(FloatCurve=DataInterface() {propertyValues {Curve {
        defaultValue=1
        keys [
         {tangentMode=RCTM_User, time=0, value=1}
         {arriveTangent=-0.0405, interpMode=RCIM_Cubic, leaveTangent=-0.0405, tangentMode=RCTM_User, time=0.7009, value=1}
         {arriveTangent=-8.0089, interpMode=RCIM_Cubic, leaveTangent=-8.0089, tangentMode=RCTM_User, time=1, value=0}
        ]
       }}})
    Scale RGB=&VectorFromFloat(Value=&FloatFromCurve(
      FloatCurve=DataInterface() {propertyValues {Curve {keys [
          {interpMode=RCIM_Cubic, tangentMode=RCTM_User, time=0, value=2}
          {arriveTangent=-1.2952, tangentMode=RCTM_User, time=0.7622, value=1.3647}
          {interpMode=RCIM_Cubic, tangentMode=RCTM_User, time=1, value=0}
         ]}}}
      Scale Curve=&MaskFloatBySpawnGroup(Masked Float=&RandomRangeFloat(Minimum=3, Maximum=4), Passthrough Float=4, Spawn Group Mask=4)
     ))
    ScaleA=true
    ScaleRGB=true
   ) {module=ScaleColor002}
   &DynamicMaterialParameters(Module.Param0WriteEnabled=true, Module.Param1WriteEnabled=true, Module.Param2WriteEnabled=true, Module.Param3WriteEnabled=true)
  ]
  rendererClasses [&NiagaraMeshRendererProperties]
  renderers [
   {rendererClass=&NiagaraMeshRendererProperties, rendererIndex=0}
   {rendererClass=&NiagaraMeshRendererProperties, rendererIndex=1}
  ]
 }
 Multidirectional_Glow001 {
  emitterUpdateScript [
   &EmitterState(
    Inactive Response=::NewEnumerator1
    Life Cycle Mode=::NewEnumerator1
    Loop Behavior=::NewEnumerator1
    Loop Duration Mode=::NewEnumerator1
    UseLoopDelay=false
   )
   &SpawnBurst_Instantaneous(Spawn Time=0.15, Use Spawn Probability=false, UseLoopCountLimit=false)
   &SpawnBurst_Instantaneous(Spawn Group=1, Spawn Time=0.15, Use Spawn Probability=false, UseLoopCountLimit=false) {module=SpawnBurst_Instantaneous002}
   &SpawnBurst_Instantaneous(Spawn Group=2, Spawn Time=0.15, Use Spawn Probability=false, UseLoopCountLimit=false) {module=SpawnBurst_Instantaneous001}
   &SpawnBurst_Instantaneous(Spawn Group=3, Spawn Time=0.15, Use Spawn Probability=false, UseLoopCountLimit=false) {module=SpawnBurst_Instantaneous004}
   &SpawnBurst_Instantaneous(Spawn Group=4, Spawn Time=0.2, Use Spawn Probability=false, UseLoopCountLimit=false) {module=SpawnBurst_Instantaneous003}
  ]
  particleSpawnScript [
   &SkeletalMeshLocation(
    Bone / Socket Index=5
    Bone Sampling Mode=::NewEnumerator4
    MaskSpawnGroup=false
    Sampled Position Offset=&MaskVectorBySpawnGroup(Masked Vector=<0 0 1.8>, Passthrough Vector=<0 0 2>, Spawn Group Mask=4)
    Sampled Position Offset Coordinate Space:ENiagaraSkelMeshTransforms=::NewEnumerator2
    Skeletal Mesh=$System.Skeletal Mesh
   )
   &InitializeParticle(
    Color=&MaskLinearColorBySpawnGroup(Masked Linear Color=<rgba 1 1 1 0.35>, Passthrough Linear Color=<rgba 1 1 1 0.35>, Spawn Group Mask=4)
    Lifetime=&MaskFloatBySpawnGroup(Masked Float=6.8, Passthrough Float=7, Spawn Group Mask=4)
    Mesh Scale=<1.05 1 1>
    Mesh Scale Mode=::NewEnumerator4
    Position Offset=&MaskVectorBySpawnGroup(
     Masked Vector=&AddVectorToPosition(Position=$Particles.SkeletalMeshLocation.SampledPosition, Vector=<0 -25 6>)
     Passthrough Vector=&AddVectorToPosition(
      Position=$Particles.SkeletalMeshLocation.SampledPosition
      Vector=&MaskVectorBySpawnGroup(Masked Vector=<0 20 10>, Passthrough Vector=<0 20 0>, Spawn Group Mask=3)
     )
     Spawn Group Mask=4
    )
    Sprite Size=&MergeHelper_Vector2Switch(0=<140 40>, 1=<140 40>, 2=<140 140>, 3=?, 4=?, Index=$Particles.SpawnGroup, Num=::NewEnumerator1)
    Sprite Size Mode=::NewEnumerator4
    UsePositionOffset=true
    Write Material Random=true
   )
   &CameraOffset(Camera Offset Amount=&MaskFloatBySpawnGroup(Masked Float=0, Passthrough Float=2.5, Spawn Group Mask=4))
   &SpriteFacingAndAlignment(
    Sprite Alignment=&MergeHelper_VectorSwitch(0=<1 0 0>, 1=<0 0 -1>, 2=<1 0 0>, 3=?, Index=$Particles.SpawnGroup, Num=::NewEnumerator1)
    Sprite Facing=&MergeHelper_VectorSwitch(0=<1.5 0 1>, 1=<0 1 0>, 2=<0 0 1>, 3=?, Index=$Particles.SpawnGroup, Num=::NewEnumerator1)
    WriteAlignment=true
    WriteFacing=true
   )
   /SetVariables(Particles.VisibilityTag=/Niagara/DynamicInputs/Helpers/MaskInt(Mask Source:NiagaraInt32=0, Mask Parameter:NiagaraInt32=?, Mask:NiagaraInt32=4, Masked Int:NiagaraInt32=1, Passthrough Int:NiagaraInt32=0)) {isSetParameters=true}
  ]
  particleUpdateScript [
   &SkeletalMeshLocation(
    Bone / Socket Index=5
    Bone Sampling Mode=::NewEnumerator4
    MaskSpawnGroup=false
    Position Sampling=::NewEnumerator2
    Sampled Position Offset=<0 -15 6>
    Sampled Position Offset Coordinate Space:ENiagaraSkelMeshTransforms=::NewEnumerator2
    Skeletal Mesh=$System.Skeletal Mesh
   ) {module=SkeletalMeshLocation001}
   /SetVariables(Particles.Position=&MaskVectorBySpawnGroup(
     Masked Vector=Linked() {linkedVariable {name=Particles.SkeletalMeshLocation001.SampledPosition, type=NiagaraPosition}}
     Passthrough Vector=Linked() {linkedVariable {name=Particles.Position, type=NiagaraPosition}}
     Spawn Group Mask=4
    )) {isSetParameters=true}
   &ParticleState()
   &ScaleColor(
    Scale Alpha=&MaskFloatBySpawnGroup(
     Masked Float=&FloatFromCurve(
      FloatCurve=DataInterface() {propertyValues {Curve {keys [
          {arriveTangent=0.222, interpMode=RCIM_Cubic, leaveTangent=0.222, tangentMode=RCTM_User, time=0, value=0}
          {arriveTangent=0.222, interpMode=RCIM_Cubic, leaveTangent=0.222, tangentMode=RCTM_User, time=0.1398, value=0.7568}
          {arriveTangent=0.0426, interpMode=RCIM_Cubic, leaveTangent=0.0426, tangentMode=RCTM_User, time=0.8009, value=0.7578}
          {arriveTangent=-1.059, interpMode=RCIM_Cubic, leaveTangent=-1.059, tangentMode=RCTM_User, time=1, value=0}
         ]}}}
      Scale Curve=&MakeCustomFloatFromBool(Bool=$User.bIsFrontEndPreview, True Float=0.6, False Float=0.75)
     )
     Passthrough Float=&FloatFromCurve(
      FloatCurve=DataInterface() {propertyValues {Curve {keys [
          {arriveTangent=10.6482, interpMode=RCIM_Cubic, leaveTangent=10.6482, tangentMode=RCTM_User, time=0, value=0}
          {arriveTangent=0.0197, interpMode=RCIM_Cubic, tangentMode=RCTM_Break, time=0.2946, value=1.0141}
          {arriveTangent=-5.1952, interpMode=RCIM_Cubic, leaveTangent=-5.1952, tangentMode=RCTM_User, time=1, value=0}
         ]}}}
      Scale Curve=&MergeHelper_FloatSwitch(
       0=&MakeCustomFloatFromBool(Bool=$User.bIsFrontEndPreview, True Float=0.2, False Float=0.5)
       1=&MakeCustomFloatFromBool(Bool=$User.bIsFrontEndPreview, True Float=0.2, False Float=0.5)
       2=&MakeCustomFloatFromBool(Bool=$User.bIsFrontEndPreview, True Float=0.2, False Float=0.5)
       3=?
       Index=$Particles.SpawnGroup
       Num=::NewEnumerator1
      )
     )
     Spawn Group Mask=4
    )
    Scale RGB=&Multiply_VectorByFloat(
     Vector=&Multiply_Vector(
      A=&MakeVectorFromLinearColorRGB(Color=$System.TODColor)
      B=&MakeVectorFromLinearColorRGB(Color=&LerpLinearColors(StartColor=<rgba 0.0515 1 0.7773 1>, EndColor=<rgba 0.3649 0.0515 1 1>, LerpFactor=$User.ColorIndex))
     )
     Float=&MaskFloatBySpawnGroup(
      Masked Float=&FNTimeOfDayScalar(Night Scale=5, Morning Scale=5, Day Scale=4, Evening Scale=4)
      Passthrough Float=&FNTimeOfDayScalar(Night Scale=5, Morning Scale=5, Day Scale=1.25, Evening Scale=4)
      Spawn Group Mask=4
     )
    )
    ScaleA=true
    ScaleRGB=true
   )
   &ScaleMeshSize()
   &FN_Light_Attributes(
    Color=&MultiplyLinearColor(
     Linear Color=&LerpLinearColors(StartColor=<rgba 0.0515 1 0.7773 1>, EndColor=<rgba 0.3649 0.0515 1 1>, LerpFactor=$User.ColorIndex)
     Scale Factor=&MakeVector4_FromLinearColor(LinearColor=$System.TODColor)
    )
    Create Color Attribute=true
    Create Exponent Attribute=true
    Create Position Attribute=true
    Create Radius Attribute=true
    Exponent=2
    Position=&AddVectorToPosition(Position=$Particles.SkeletalMeshLocation.SampledPosition, Vector=<0 0 0>)
    Radius=30
    Scale Color=&Multiply_Float(
     A=&MakeCustomFloatFromBool(Bool=$User.bIsFrontEndPreview, True Float=3, False Float=5)
     B=&FNTimeOfDayScalar(Night Scale=&Lerp_Float(A=1, B=2.5, Alpha=$User.ColorIndex), Morning Scale=1, Evening Scale=1)
    )
   )
   &SolveForcesAndVelocity(Clamp Velocity=false, Limit Acceleration=false)
  ]
  rendererClasses [
   &NiagaraSpriteRendererProperties
   &NiagaraLightRendererProperties
   &NiagaraMeshRendererProperties
  ]
  renderers [
   {rendererClass=&NiagaraSpriteRendererProperties}
   {rendererClass=&NiagaraLightRendererProperties, rendererIndex=1}
   {rendererClass=&NiagaraMeshRendererProperties, rendererIndex=2}
  ]
 }
 Multidirectional_Glow_DetailsEpic {
  emitterUpdateScript [
   &EmitterState(
    Inactive Response=::NewEnumerator1
    Life Cycle Mode=::NewEnumerator1
    Loop Behavior=::NewEnumerator1
    Loop Duration Mode=::NewEnumerator1
    UseLoopDelay=false
   )
   &SpawnBurst_Instantaneous(Spawn Count=5, Spawn Time=0.3, Use Spawn Probability=false, UseLoopCountLimit=false)
  ]
  particleSpawnScript [
   &SkeletalMeshLocation(Bone / Socket Index=5, Bone Sampling Mode=::NewEnumerator4, MaskSpawnGroup=false, Skeletal Mesh=$System.Skeletal Mesh)
   &FN_SetSpawnGroup(Number of Spawn Groups=5)
   &InitializeParticle(
    Color=<rgba 1 1 1 0.35>
    Lifetime=6
    Position Offset=&AddVectorToPosition(Position=$Particles.SkeletalMeshLocation.SampledPosition, Vector=<0 30 5>)
    Position Offset Coordinate Space=::Local
    Sprite Size=&MergeHelper_Vector2Switch(0=<100 55>, 1=<125 40>, 2=<150 55>, 3=?, 4=?, Index=$Particles.SpawnGroup, Num=::NewEnumerator1)
    Sprite Size Mode=::NewEnumerator4
    UsePositionOffset=true
    Write Material Random=true
   )
   &CameraOffset(Camera Offset Amount=2.5)
   &SpriteFacingAndAlignment(
    Sprite Alignment=&MergeHelper_VectorSwitch(0=<1 0 0>, 1=<0 0 1>, 2=<1 0 0>, 3=?, Index=$Particles.SpawnGroup, Num=::NewEnumerator1)
    Sprite Facing=&MergeHelper_VectorSwitch(0=<1.5 0 1>, 1=<0 1 0>, 2=<-1.5 0 1>, 3=?, Index=$Particles.SpawnGroup, Num=::NewEnumerator1)
    WriteAlignment=true
    WriteFacing=true
   )
  ]
  particleUpdateScript [
   &ParticleState()
   &ScaleColor(
    Scale Alpha=&FloatFromCurve(
     FloatCurve=DataInterface() {propertyValues {Curve {keys [<at 0 0.8 RCIM_Cubic>,<at 0.2946 1.0141 RCIM_Cubic>,<at 1 0.8 RCIM_Cubic>]}}}
     Scale Curve=&MergeHelper_FloatSwitch(0=0.6, 1=1.75, 2=0.6, 3=?, Index=$Particles.SpawnGroup, Num=::NewEnumerator1)
    )
    Scale RGB=&Multiply_VectorByFloat(
     Vector=&Multiply_Vector(
      A=&MakeVectorFromLinearColorRGB(Color=$System.TODColor)
      B=&MakeVectorFromLinearColorRGB(Color=&LerpLinearColors(StartColor=<rgba 0.0515 1 0.7773 1>, EndColor=<rgba 0.3649 0.0515 1 1>, LerpFactor=$User.ColorIndex))
     )
     Float=&FNTimeOfDayScalar(Night Scale=1.25, Morning Scale=1.25, Day Scale=1.25, Evening Scale=1.25)
    )
    ScaleA=true
    ScaleRGB=true
   )
   &ScaleColor(
    Scale Alpha=&FloatFromCurve(
     FloatCurve=DataInterface() {propertyValues {Curve {keys [
         {arriveTangent=10.6482, interpMode=RCIM_Cubic, leaveTangent=10.6482, tangentMode=RCTM_User, time=0, value=0}
         {arriveTangent=0.0197, interpMode=RCIM_Cubic, tangentMode=RCTM_Break, time=0.2946, value=1.0141}
         {arriveTangent=-5.1952, interpMode=RCIM_Cubic, leaveTangent=-5.1952, tangentMode=RCTM_User, time=1, value=0}
        ]}}}
     Scale Curve=&MakeCustomFloatFromBool(Bool=$User.bIsFrontEndPreview, True Float=0.3, False Float=0.7)
    )
    ScaleA=true
    ScaleRGB=true
   ) {module=ScaleColor001}
  ]
  rendererClasses [&NiagaraSpriteRendererProperties]
  renderers [{rendererClass=&NiagaraSpriteRendererProperties}]
 }
 Wave_Burst {
  emitterUpdateScript [
   &EmitterState(
    Inactive Response=::NewEnumerator1
    Life Cycle Mode=::NewEnumerator1
    Loop Behavior=::NewEnumerator1
    Loop Delay=3.9
    Loop Duration=2
    UseLoopDelay=true
   )
   &SpawnBurst_Instantaneous(Use Spawn Probability=false, UseLoopCountLimit=false) {module=SpawnBurst_Instantaneous001}
   &SpawnBurst_Instantaneous(Spawn Group=1, Use Spawn Probability=false, UseLoopCountLimit=false)
  ]
  particleSpawnScript [
   &InitializeParticle(
    Color Mode=::NewEnumerator0
    Lifetime=&MaskFloatBySpawnGroup(Masked Float=2, Passthrough Float=1.5)
    Mesh Scale=&MaskVectorBySpawnGroup(Masked Vector=<0.5 0.5 0.5>, Passthrough Vector=<0.5 0.5 1>)
    Mesh Scale Mode=::NewEnumerator4
    UsePositionOffset=true
    Write Material Random=true
   )
   &SkeletalMeshLocation(
    Bone / Socket Index=4
    Bone Sampling Mode=::NewEnumerator4
    MaskSpawnGroup=false
    Sampled Position Offset=<0 0 -7>
    Skeletal Mesh=$System.Skeletal Mesh
    Use Sampled Mesh Attribute as Ribbon ID=false
   ) {module=SkeletalMeshLocation001}
   /SetVariables(Particles.VisibilityTag=&MaskIntBySpawnGroup(Masked Int=0, Passthrough Int=1, Spawn Group Mask=0)) {isSetParameters=true}
  ]
  particleUpdateScript [
   &ParticleState()
   &ScaleMeshSize(Scale Factor=&MaskVectorBySpawnGroup(
     Masked Vector=&VectorFromFloat(Value=0.55)
     Passthrough Vector=&VectorFromFloat(Value=&FloatFromCurve(
       FloatCurve=DataInterface() {propertyValues {Curve {keys [<at 0 0 RCIM_Cubic>, {arriveTangent=1, interpMode=RCIM_Cubic, leaveTangent=1, time=0.2041, value=0.6997}, <at 1 1 RCIM_Cubic>]}}}
       Scale Curve=0.9
      ))
    ))
   /SetVariables(Particles.Scale002:Vector3f=&Multiply_Vector(
     A=<0.6 0.6 0.1>
     B=&VectorFromFloat(Value=&FloatFromCurve(
       FloatCurve=DataInterface() {propertyValues {Curve {keys [
           {arriveTangent=1.0263, interpMode=RCIM_Cubic, leaveTangent=1.0263, tangentMode=RCTM_User, time=0, value=1}
           {arriveTangent=0.333, interpMode=RCIM_Cubic, leaveTangent=0.333, tangentMode=RCTM_User, time=0.2019, value=1.199}
           {arriveTangent=0.018, interpMode=RCIM_Cubic, leaveTangent=0.018, tangentMode=RCTM_User, time=1, value=1.325}
          ]}}}
       Scale Curve=0.5
      ))
    )) {isSetParameters=true}
   &DynamicMaterialParameters(
    Index 0 Param 1=&FloatFromCurve(FloatCurve=DataInterface() {propertyValues {Curve {keys [
         {arriveTangent=-4.6269, interpMode=RCIM_Cubic, leaveTangent=-4.6269, tangentMode=RCTM_User, time=0, value=0.5}
         {arriveTangent=-1.721, interpMode=RCIM_Cubic, leaveTangent=-1.721, tangentMode=RCTM_User, time=0.2068, value=-0.2953}
         {arriveTangent=-0.376, interpMode=RCIM_Cubic, leaveTangent=-0.376, tangentMode=RCTM_User, time=1, value=-1}
        ]}}})
    Module.Index0Param0Write=true
    Module.Index0Param1Write=true
    Module.Index0Param2Write=true
    Module.Index0Param3Write=true
   )
   &ScaleColor(
    Scale Alpha=&MaskFloatBySpawnGroup(
     Masked Float=&FloatFromCurve(FloatCurve=DataInterface() {propertyValues {Curve {keys [
          {arriveTangent=-0.024, interpMode=RCIM_Cubic, leaveTangent=-0.024, tangentMode=RCTM_User, time=0, value=0}
          {arriveTangent=-1.4602, interpMode=RCIM_Cubic, leaveTangent=-1.4602, tangentMode=RCTM_User, time=0.1, value=1}
          <at 1 0 RCIM_Cubic>
         ]}}})
     Passthrough Float=&FloatFromCurve(
      FloatCurve=DataInterface() {propertyValues {Curve {keys [
          {arriveTangent=-1.4862, interpMode=RCIM_Cubic, leaveTangent=-1.4862, tangentMode=RCTM_User, time=0, value=1}
          {arriveTangent=-0.3018, interpMode=RCIM_Cubic, leaveTangent=-0.3018, tangentMode=RCTM_User, time=1, value=0}
         ]}}}
      Scale Curve=&FN_PickFloatDependOnPlatform(TargetPlatformActive=1, TargetPlatformNoActive=0.6)
     )
    )
    ScaleA=true
    ScaleRGB=true
   )
   /SetVariables(Particles.Color002=&MakeLinearColorFromVectorAndFloat(
     "Vector (RGB)"=&MakeVectorFromLinearColorRGB(Color=<rgba 1 1 1 1>)
     "Float (Alpha)"=&FloatFromCurve(FloatCurve=DataInterface() {propertyValues {Curve {keys [
          {arriveTangent=-0.024, interpMode=RCIM_Cubic, leaveTangent=-0.024, tangentMode=RCTM_User, time=0, value=0}
          {arriveTangent=-0.3677, interpMode=RCIM_Cubic, leaveTangent=-0.3677, tangentMode=RCTM_User, time=0.1, value=1}
          {arriveTangent=-1.5833, interpMode=RCIM_Cubic, leaveTangent=-1.5833, time=0.3, value=0.4}
          {arriveTangent=-0.3356, interpMode=RCIM_Cubic, leaveTangent=-0.3356, tangentMode=RCTM_User, time=0.7, value=0.05}
          <at 1 0 RCIM_Cubic>
         ]}}})
    )) {isSetParameters=true}
   /SetVariables(Particles.Color003:LinearColor=&MakeLinearColorFromVectorAndFloat(
     "Vector (RGB)"=&MakeVectorFromLinearColorRGB(Color=<rgba 1 1 1 1>)
     "Float (Alpha)"=&FloatFromCurve(
      FloatCurve=DataInterface() {propertyValues {Curve {keys [
          {arriveTangent=-0.024, interpMode=RCIM_Cubic, leaveTangent=-0.024, tangentMode=RCTM_User, time=0, value=0}
          {arriveTangent=-0.3677, interpMode=RCIM_Cubic, leaveTangent=-0.3677, tangentMode=RCTM_User, time=0.1, value=1}
          {arriveTangent=-1.5, interpMode=RCIM_Cubic, leaveTangent=-1.5, time=0.3, value=0.5}
          {arriveTangent=-0.5105, interpMode=RCIM_Cubic, leaveTangent=-0.5105, tangentMode=RCTM_User, time=0.7, value=0.1}
          <at 1 0 RCIM_Cubic>
         ]}}}
      Scale Curve=0.3
     )
    )) {isSetParameters=true}
   /SetVariables(
    Particles.Dyn:Vector4f=&MakeVector4(
     X=&FloatFromCurve(FloatCurve=DataInterface() {propertyValues {Curve {keys [
          {arriveTangent=-3.2904, interpMode=RCIM_Cubic, leaveTangent=-3.2904, tangentMode=RCTM_User, time=0, value=0.7}
          {arriveTangent=-2.1713, interpMode=RCIM_Cubic, leaveTangent=-2.1713, tangentMode=RCTM_User, time=0.259, value=-0.2031}
          {arriveTangent=-0.376, interpMode=RCIM_Cubic, leaveTangent=-0.376, tangentMode=RCTM_User, time=1, value=-1}
         ]}}})
     Y=0
     Z=0
     W=0
    )
    Particles.Dyn 2001:Vector4f=&MakeVector4(
     X=&FN_PickFloatDependOnPlatform(TargetPlatformActive=180, TargetPlatformNoActive=100)
     Y=&FN_PickFloatDependOnPlatform(TargetPlatformActive=2.5, TargetPlatformNoActive=5)
     Z=0
     W=0
    )
    Particles.Dyn 2_0:Vector4f=&MakeVector4(
     X=&FloatFromCurve(FloatCurve=DataInterface() {propertyValues {Curve {keys [<at 0 1>,<at 0.2 1.375>,<at 1 0>]}}}, Scale Curve=1.2)
     Y=2
     Z=&FloatFromCurve(
      FloatCurve=DataInterface() {propertyValues {Curve {keys [<at 0.1 1>,<at 1 0>]}}}
      Scale Curve=&FN_FrontEndFloatModifier(Gameplay=6, Front End=3)
     )
     W=&FloatFromCurve(FloatCurve=DataInterface() {propertyValues {Curve {keys [
          <at 0 1>
          {arriveTangent=0.3079, interpMode=RCIM_Cubic, leaveTangent=0.3079, tangentMode=RCTM_User, time=0.3, value=0.9}
          <at 1 1 RCIM_Cubic>
         ]}}})
    )
   ) {isSetParameters=true}
  ]
  rendererClasses [&NiagaraMeshRendererProperties]
  renderers [
   {rendererClass=&NiagaraMeshRendererProperties, rendererIndex=0}
   {rendererClass=&NiagaraMeshRendererProperties, rendererIndex=1}
   {rendererClass=&NiagaraMeshRendererProperties, rendererIndex=2}
  ]
 }
}
summary {
 systemName=NS_Emote_SwoopClasp
 emitters [
  {emitterName=Multidirectional_Glow_DetailsEpic, rendererClasses [&NiagaraSpriteRendererProperties]}
  {emitterName=Wave_Burst, rendererClasses [&NiagaraMeshRendererProperties]}
  {
   emitterName=Debris
   rendererClasses [
    &NiagaraSpriteRendererProperties
    &NiagaraLightRendererProperties
    &NiagaraMeshRendererProperties
   ]
  }
  {
   emitterName=Multidirectional_Glow001
   rendererClasses [
    &NiagaraSpriteRendererProperties
    &NiagaraLightRendererProperties
    &NiagaraMeshRendererProperties
   ]
  }
  {emitterName=Line_Burst001, rendererClasses [&NiagaraMeshRendererProperties]}
  {emitterName=Base, rendererClasses [&NiagaraSpriteRendererProperties]}
 ]
 userVariables [
  User.bIsFrontEnd {type=NiagaraBool}
  User.bIsFrontEndPreview {type=NiagaraBool}
  User.ColorIndex {type=NiagaraFloat, defaultValue=0, description=""}
  User.FrontEnd_color {type=LinearColor, description="", defaultValue {struct=LinearColor, value=<rgba 1 0.6743 0.6445 1>}}
  User.FrontEndScale {type=NiagaraFloat, defaultValue=0.4, description=""}
  User.HueShift {type=NiagaraFloat, defaultValue=0, description=""}
  User.NormalizedNotifyProgress {type=NiagaraFloat, defaultValue=0, description=""}
 ]
}
